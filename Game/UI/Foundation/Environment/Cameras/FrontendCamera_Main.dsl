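; Frontend camera Blueprint for the main menu that triggers a post-process transition whenever it becomes active.
; On activation it gathers every TVPostProcessBP actor in the level and calls FrontEndCameraSwitch on each, playing the camera-switch post-process animation.
; The construction script is empty.
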(event Camera|Fortnite|EventOnActivated
  (params (PlayerController :type Fort%20Player%20Controller%20Object%20Reference))
  ;[0.5688 1 0] Camera switch post process anim
  (bind _outactors (Actor|GetAllActorsOfClass "/Game/UI/Foundation/Environment/Cameras/TVPostProcessBP.TVPostProcessBP_C"))
  (for _array_element _outactors
    (Class|TVPostProcessBP|FrontEndCameraSwitch _array_element)))

(fn ConstructionScript)
