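; Frontend camera Blueprint for the Manage view that drives screen effects and character lighting when the camera becomes active or inactive.
; On activation it calls FrontEndCameraSwitch on every TVPostProcessBP actor to play the camera switch post-process animation, then turns on the vault character lights.
; On deactivation it turns those lights off.
; Actors are found by class on each event, and only the first VaultCharacterLightingBP found is toggled.
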(event Camera|Fortnite|EventOnActivated
  (params (PlayerController :type Fort%20Player%20Controller%20Object%20Reference))
  ;[0.5688 1 0] Camera switch post process anim
  (bind _outactors (Actor|GetAllActorsOfClass "/Game/UI/Foundation/Environment/Cameras/TVPostProcessBP.TVPostProcessBP_C"))
  (for _array_element _outactors
    (Class|TVPostProcessBP|FrontEndCameraSwitch _array_element))
  ;[1 1 1] Turn on lights
  (bind _outactors_2 (Actor|GetAllActorsOfClass "/Game/Environments/FrontEnd/Blueprints/VaultCharacterLightingBP.VaultCharacterLightingBP_C"))
  (Class|VaultCharacterLightingBP|LightControl (Utilities|Array|Get(acopy) _outactors_2) true))

(event Camera|Fortnite|EventOnDeactivated
  (params (PlayerController :type Fort%20Player%20Controller%20Object%20Reference))
  ;[1 1 1] Turn off lights
  (bind _outactors (Actor|GetAllActorsOfClass "/Game/Environments/FrontEnd/Blueprints/VaultCharacterLightingBP.VaultCharacterLightingBP_C"))
  (Class|VaultCharacterLightingBP|LightControl (Utilities|Array|Get(acopy) _outactors) false))

(fn ConstructionScript)
