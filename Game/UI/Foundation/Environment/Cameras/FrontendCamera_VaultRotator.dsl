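; Frontend camera blueprint for vault item-preview rotators that coordinates a switch effect and rotator lighting.
; On activation it triggers the camera-switch effect on every TVPostProcessBP actor, and on deactivation it turns off the active rotator's lights.
; SetVaultRotatorLighting searches all VaultWorld actors by tag, stores and lights the matching actor's VaultRotator, prints its name, and turns off the others.
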(event Camera|Fortnite|EventOnActivated
  (params (PlayerController :type Fort%20Player%20Controller%20Object%20Reference))
  ;[0.5688 1 0] Camera switch post process anim
  (bind _outactors (Actor|GetAllActorsOfClass "/Game/UI/Foundation/Environment/Cameras/TVPostProcessBP.TVPostProcessBP_C"))
  (for _array_element _outactors
    (Class|TVPostProcessBP|FrontEndCameraSwitch _array_element)))

(event Camera|Fortnite|EventOnDeactivated
  (params (PlayerController :type Fort%20Player%20Controller%20Object%20Reference))
  (Class|VaultRotator|LightControl (Variables|Default|GetActiveVaultRotator) false))

(fn SetVaultRotatorLighting
  (params (ItemRotatorTag :type Name))
  (bind _outactors (Actor|GetAllActorsOfClass "/Game/Environments/FrontEnd/Blueprints/ItemPreview/VaultWorld.VaultWorld_C"))
  (for _array_element _outactors
    (bind _vaultrotator (Class|VaultWorld|GetVaultRotator _array_element))
    (if (Actor|ActorHasTag _array_element ItemRotatorTag)
      (bind _output_get (Variables|Default|SetActiveVaultRotator _vaultrotator))
      (Development|PrintString (Utilities|GetDisplayName _vaultrotator))
      (Class|VaultRotator|LightControl _output_get true)
      (else
        (Class|VaultRotator|LightControl _vaultrotator false)))))

(fn ConstructionScript)
