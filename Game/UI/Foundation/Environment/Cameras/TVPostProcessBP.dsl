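; TVPostProcessBP is a Blueprint skeleton for TV post-process effects that respond to front-end camera switches.
; FrontEndCameraSwitch calls ExecuteCameraSwitch only when IsEnabledForCurrentSubgame passes, and that function always returns false here.
; The construction script calls Camera_DisableEffects when DisableOnStartup is set.
; ExecuteCameraSwitch and Camera_DisableEffects are empty events, so the file itself implements no visual effect.
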(event Custom|FrontEndCameraSwitch
  ;[1 0.75 0] FrontEndCameraSwitch
  (if (CallFunction|IsEnabledforCurrentSubgame)
    (CallFunction|ExecuteCameraSwitch)))

(event Custom|ExecuteCameraSwitch)

(event Custom|Camera_DisableEffects)

(fn IsEnabledForCurrentSubgame
  (return false))

(fn ConstructionScript
  (if (Variables|Default|GetDisableOnStartup)
    (CallFunction|CameraDisableEffects)))
